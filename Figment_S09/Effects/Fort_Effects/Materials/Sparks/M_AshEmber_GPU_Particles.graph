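# Unlit translucent particle sprite material for ash and ember sparks, with responsive AA enabled.
# An 8x8 husk-bits SubUV flipbook advances by particle relative time, offset per particle by a random value and scaled by Frame Speed.
# Texture channels are recombined and tinted by particle color, and alpha is multiplied by particle alpha with an optional static-switched near-camera fade.
# ES2 and ES3.1 paths sample a single flipbook cell with simpler channel math.
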Material /Figment_S09/Effects/Fort_Effects/Materials/Sparks/M_AshEmber_GPU_Particles
graph {
 properties {
  bEnableResponsiveAA=true
  bUsedWithParticleSprites=true
  blendMode=BLEND_Translucent
  pixelDepthOffsetMode=PDOM_Legacy
  shadingModel=MSM_Unlit
  usedShadingModels=Unlit
 }
 nodes [
  n1 = TextureObject() {texture=/Game/Effects/Fort_Effects/Textures/Husk/T_Husk_Bits_SubUV}
  n2 = ParticleRelativeTime()
  n3 = ParticleRandom()
  n4 = MaterialFunctionCall!(
   Texture=@n1
   UVs
   SubImages=Constant2Vector() {g=8, r=8}
   Frame=Multiply(A=Add(A=@n2, B=@n3), B=ScalarParameter() {defaultValue=64, parameterName=Frame Speed})
  ) {materialFunction=/Engine/Functions/Engine_MaterialFunctions03/Texturing/SubUV_Function}
  n5 = ParticleColor()
  n6 = ComponentMask(None=@n4.RGB) {b=true}
  n7 = Multiply(A=ComponentMask(None=@n4.RGB) {g=true}, B=Constant() {r=0.1})
  n8 = Add(A=ComponentMask(None=@n4.RGB) {r=true}, B=Multiply(A=@n5.RGB, B=Add(A=@n6, B=@n7)))
  n9 = TextureSample(UVs=TextureCoordinate() {uTiling=0.125, vTiling=0.125}, Tex=@n1) {texture=&T_Husk_Bits_SubUV}
  n10 = Multiply(A=@n5.RGB, B=Add(A=@n9.G, B=Multiply(A=@n9.B) {constB=0.2}))
  FeatureLevelSwitch(Default=@n8, ES2=@n10, ES3_1=@n10, SM4_REMOVED=@n8)
  n11 = Multiply(A=@n5.A, B=@n4.Alpha)
  n12 = Multiply(A=@n5.A, B=@n9.A)
  n13 = FeatureLevelSwitch(Default=@n11, ES2=@n12, ES3_1=@n12, SM4_REMOVED=@n11)
  n14 = ScalarParameter() {defaultValue=250, parameterName=Near Camera Distance}
  StaticSwitchParameter(
   True=Multiply(A=@n13, B=MaterialFunctionCall(Distance=@n14) {materialFunction=&Near_Camera_Fade}.Result)
   False=@n13
  ) {parameterName=Use Near Camer Fade}
 ]
}
